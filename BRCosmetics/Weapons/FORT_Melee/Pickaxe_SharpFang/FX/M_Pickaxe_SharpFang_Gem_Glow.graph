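# Masked, lit material for Niagara mesh particles that renders the glowing gem of the SharpFang pickaxe.
# It adds the emissive texture, tinted by particle color, to the diffuse texture, alongside normal and mask textures.
# A dimmed particle color feeds the shared MF_FX_EmissiveAdjustments function.
# A Step compares the Gem Mask dynamic parameter with inverted second-UV V, giving particles a hard-edged cutoff along the mesh.
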Material /BRCosmetics/Weapons/FORT_Melee/Pickaxe_SharpFang/FX/M_Pickaxe_SharpFang_Gem_Glow
graph {
 properties {bUsedWithNiagaraMeshParticles=true, blendMode=BLEND_Masked, pixelDepthOffsetMode=PDOM_Legacy, usedShadingModels=Default Lit}
 nodes [
  n1 = ParticleColor()
  Add(
   A=TextureSample() {texture=/BRCosmetics/Weapons/FORT_Melee/Pickaxe_SharpFang/Textures/T_SharpFang_D}.RGB
   B=Multiply(A=TextureSample() {texture=/BRCosmetics/Weapons/FORT_Melee/Pickaxe_SharpFang/Textures/T_SharpFang_E}.RGB, B=@n1.RGB)
  )
  Constant() {r=0.4}
  MaterialFunctionCall(Emissive=Multiply(A=@n1.RGB) {constB=0.2}) {materialFunction=&MF_FX_EmissiveAdjustments}
  n2 = TextureCoordinate() {coordinateIndex=1}
  Step(
   Y=OneMinus(None=ComponentMask(None=@n2) {g=true})
   X=DynamicParameter() {defaultValue=<rgba 0 1 1 1>, paramNames [Gem Mask, Param2, Param3, Param4]}.Gem Mask
  ) {constX=0}
  TextureSample() {samplerType=SAMPLERTYPE_Normal, texture=/BRCosmetics/Weapons/FORT_Melee/Pickaxe_SharpFang/Textures/SharpFang_N}
  TextureSample() {samplerType=SAMPLERTYPE_Masks, texture=/BRCosmetics/Weapons/FORT_Melee/Pickaxe_SharpFang/Textures/T_SharpFang_S}
 ]
}
