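# Translucent, unlit, two-sided Niagara ribbon and mesh-particle material for the SharpFang pickaxe effect, tinted by particle color.
# Opacity combines noise-distorted plasma and UV edge masks with a rounded caustics dissolve, depth fade and TrailRange sphere mask.
# World position offset adds scrolling sine waves to a switchable, clamped pull toward a Niagara-set position, weighted along the UV length.
# Dynamic parameters drive wave offset, amplitude, range and dissolve.
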Material /BRCosmetics/Weapons/FORT_Melee/Pickaxe_SharpFang/FX/M_Pickaxe_SharpFang_Ribbon
graph {
 properties {
  bUsedWithNiagaraMeshParticles=true
  bUsedWithNiagaraRibbons=true
  blendMode=BLEND_Translucent
  pixelDepthOffsetMode=PDOM_Legacy
  shadingModel=MSM_Unlit
  twoSided=true
  usedShadingModels=Unlit
 }
 nodes [
  n1 = DynamicParameter() {desc=Parameters set by Niagara, parameterIndex=1, paramNames [X,Y,Z,WPOIntensity]}
  n2 = Clamp(
   None=Subtract(
    A=MaterialFunctionCall(X=@n1.X, Y=@n1.X, Z=@n1.X) {materialFunction=&MakeFloat3}.Result
    B=ParticlePositionWS()
   )
   Min=ScalarParameter() {defaultValue=-100, group=WPO, parameterName=Clamp_min}
   Max=ScalarParameter() {defaultValue=100, group=WPO, parameterName=Clamp_max}
  ) {maxDefault=1000, minDefault=-1000}
  n3 = TextureCoordinate()
  n4 = Multiply(
   A=Power(Base=ComponentMask(None=@n3) {r=true}, Exp=ScalarParameter() {defaultValue=2.5, group=WPO, parameterName=WPOBendPower})
   B=@n1.WPOIntensity
  )
  n5 = Add(A=TextureCoordinate(), B=DynamicParameter() {defaultValue=<rgba 1 1 70 1>, paramNames [WPO Offset, WPO, TrailRange, Dis]}.WPO Offset)
  n6 = Multiply(A=Time()) {constB=-0.4}
  n7 = Sine(None=ComponentMask(None=Add(A=@n5, B=@n6)) {r=true}) {period=0.7}
  n8 = Constant3Vector() {constant=<rgba 0 0 0.5 0>}
  n9 = TextureCoordinate()
  n10 = DynamicParameter() {defaultValue=<rgba 1 1 70 1>, paramNames [WPO Offset, WPO, TrailRange, Dis]}
  n11 = Time()
  n12 = ComponentMask(None=Add(A=Add(A=@n9, B=@n10.WPO Offset), B=Multiply(A=@n11) {constB=-0.4})) {r=true}
  n13 = Multiply(A=LinearInterpolate(Alpha=Sine(None=@n12) {period=0.7}) {constA=1, constB=2}, B=Constant3Vector() {constant=<rgba 0 0.5 0 0>})
  n14 = Add(A=Transform(None=Multiply(A=@n7, B=@n8)), B=Transform(None=Multiply(A=@n13)))
  n15 = Multiply(A=Constant() {r=7}, B=DynamicParameter() {defaultValue=<rgba 1 1 70 1>, paramNames [WPO Offset, WPO, TrailRange, Dis]}.WPO)
  n16 = Multiply(A=ComponentMask(None=TextureCoordinate()) {r=true}) {constB=5}
  Add(
   A=Multiply(A=Multiply(A=@n14, B=@n15) {constB=7}, B=Saturate(None=@n16))
   B=StaticSwitchParameter(True=LinearInterpolate(B=@n2, Alpha=@n4), False=Constant()) {defaultValue=true, group=WPO, parameterName=Use_Interpolate WPO}
  )
  n17 = Add(
   A=TextureCoordinate() {vTiling=0.3}
   B=AppendVector(A=DynamicParameter() {defaultValue=<rgba 1 1 70 1>, paramNames [WPO Offset, WPO, TrailRange, Dis]}.WPO Offset, B=Constant())
  )
  n18 = ScalarParameter() {parameterName=Dissolve mask Min}
  n19 = DynamicParameter() {defaultValue=<rgba 1 1 70 1>, paramNames [WPO Offset, WPO, TrailRange, Dis]}
  n20 = Power(Base=ComponentMask(None=TextureCoordinate()) {r=true}) {constExponent=1.3}
  n21 = Power(
   Base=TextureSample(UVs=Panner(Coordinate=@n17) {speedX=-0.5}) {samplerType=SAMPLERTYPE_LinearGrayscale, texture=/Game/Global/Textures/Default/Caustics/T_Nosie_Caustics_Grayscale}.R
   Exp=Saturate(None=LinearInterpolate(A=@n18, B=@n19.Dis, Alpha=@n20) {constA=0.1, constB=2})
  )
  LinearInterpolate(Alpha=@n21) {constA=-4, constB=4}
  n22 = ParticleColor()
  MaterialFunctionCall(Emissive=@n22.RGB) {materialFunction=&MF_FX_EmissiveAdjustments}
  n23 = Sine(None=ComponentMask(None=TextureCoordinate()) {g=true}) {period=2}
  n24 = TextureSample(UVs=Panner(Coordinate=TextureCoordinate() {vTiling=0.5}) {speedX=-0.05, speedY=-0.1}) {samplerType=SAMPLERTYPE_Grayscale, texture=&LowResBlurredNoise}
  n25 = Panner(Coordinate=Add(A=TextureCoordinate() {uTiling=0.8, vTiling=0.2}, B=Multiply(A=@n24.R) {constB=0.2})) {speedX=-0.3, speedY=0.05}
  n26 = Multiply(A=Power(Base=TextureSample(UVs=@n25) {samplerType=SAMPLERTYPE_LinearGrayscale, texture=/Game/Effects/Fort_Effects/Emotes/UltraLight/Texture/T_Plasma_C_Greyscale}.R)) {constB=0.1}
  n27 = SphereMask(A=TextureCoordinate() {uTiling=0}, B=Constant2Vector() {g=0.5}) {attenuationRadius=1, hardnessPercent=0}
  n28 = Add(A=Power(Base=OneMinus(None=@n27)))
  n29 = Multiply(A=@n28) {constB=0.03}
  n30 = OneMinus(None=SphereMask(A=TextureCoordinate() {uTiling=0}, B=Constant2Vector() {g=0.5}) {attenuationRadius=1, hardnessPercent=20})
  n31 = Multiply(
   A=Saturate(None=Multiply(A=@n23) {constB=4})
   B=LinearInterpolate(A=Add(A=@n26, B=@n29), B=Saturate(None=@n28), Alpha=Power(Base=@n30) {constExponent=2.5}) {constA=0.05}
  )
  n32 = Multiply(A=ComponentMask(None=TextureCoordinate()) {r=true}, B=ScalarParameter() {defaultValue=20, parameterName=Lead Edge mask}) {constB=20}
  n33 = TextureCoordinate()
  n34 = Power(Base=OneMinus(None=ComponentMask(None=@n33) {r=true})) {constExponent=1.5}
  n35 = Multiply(A=Multiply(A=Multiply(A=@n22.A, B=@n31), B=Saturate(None=@n32)), B=Saturate(None=Multiply(A=@n34) {constB=5}))
  n36 = Saturate(None=Round(None=@n21))
  Multiply(
   A=DepthFade(Opacity=Multiply(A=@n35, B=@n36)) {fadeDistanceDefault=10}
   B=SphereMask(
    A=WorldPosition().XYZ
    B=ParticlePositionWS()
    Radius=DynamicParameter() {defaultValue=<rgba 1 1 70 1>, paramNames [WPO Offset, WPO, TrailRange, Dis]}.TrailRange
   ) {attenuationRadius=70, hardnessPercent=0}
  )
  OneMinus(None=@n3)
  ScalarParameter() {defaultValue=1, parameterName=Dissolve mask Max}
 ]
}
