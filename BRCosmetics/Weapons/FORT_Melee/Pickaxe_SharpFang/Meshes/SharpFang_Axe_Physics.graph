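# Physics asset for the SharpFang two-handed pickaxe mesh, giving its handle and dangling chain bones collision and joints for physics simulation.
# Capsule bodies cover the handle and each dyn_Chain bone from 00 through End, with several capsules on some bones.
# Constraints link the bodies serially from the handle, with wide swing limit values mid-chain and much tighter ones near both ends.
# The last joint locks swing and twist, fixing dyn_Chain_End's orientation to dyn_Chain_06.
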PhysicsAsset /BRCosmetics/Weapons/FORT_Melee/Pickaxe_SharpFang/Meshes/SharpFang_Axe_Physics
bodies {
 Melee_TwoHanded_Handle [
  {shapeType=Capsule, length=148.5734, radius=2, center=<0.6887 -0.6771 20.0697>, rotation=<rot -0.5045 -90.0036 0.0423>}
  {
   shapeType=Capsule
   length=5.75
   radius=1.25
   shapeName=Melee_TwoHanded_Handle_capsule
   center=<0.8736 -17.1511 62.3372>
   rotation=<rot 0.0807 -0.0354 -19.4121>
  }
  {shapeType=Capsule, length=0.8559, radius=3.6801, shapeName=Melee_TwoHanded_Handle_capsule, center=<0.781 -0.8816 50.2881>}
 ]
 dyn_Chain_00 [{
   shapeType=Capsule
   length=0
   radius=2.25
   shapeName=dyn_Chain_00_capsule
   center=<-0.0411 -0.0549 0.2696>
   rotation=<rot 0.0267 91.8298 -90.1845>
  }]
 dyn_Chain_01 [
  {
   shapeType=Capsule
   length=6.7713
   radius=1.8588
   shapeName=dyn_Chain_01_capsule
   center=<4.3285 0.0001 -0.3254>
   rotation=<rot -0.02 -90 -90.7143>
  }
  {shapeType=Capsule, length=8.5, radius=1.25, shapeName=dyn_Chain_01_capsule, center=<4.4747 0 0>}
  {shapeType=Capsule, length=6.0243, radius=1.4757, shapeName=dyn_Chain_01_capsule, center=<2.3886 0 1.9411>, rotation=<rot -45.0868 0 0>}
 ]
 dyn_Chain_02 [{
   shapeType=Capsule
   length=4.0214
   radius=1.8392
   shapeName=dyn_Chain_02_capsule
   center=<4.0229 0.568 0.6722>
   rotation=<rot -2.1112 -91.9845 89.4435>
  }]
 dyn_Chain_03 [{
   shapeType=Capsule
   length=3.9659
   radius=1.9262
   shapeName=dyn_Chain_03_capsule
   center=<4.5963 0.76 0.2214>
   rotation=<rot -1.3267 91.3361 -88.5708>
  }]
 dyn_Chain_04 [
  {
   shapeType=Capsule
   length=5.75
   radius=2.0095
   shapeName=dyn_Chain_04_capsule
   center=<4.0265 0.2179 0.0411>
   rotation=<rot 0.1001 -93.0222 -89.3654>
  }
  {shapeType=Capsule, length=8.75, radius=1.25, shapeName=dyn_Chain_04_capsule, center=<4.6305 0.3423 -0.5211>, rotation=<rot 0.0035 0 -43.824>}
 ]
 dyn_Chain_05 [{
   shapeType=Capsule
   length=5.0741
   radius=1.8342
   shapeName=dyn_Chain_05_capsule
   center=<3.2581 0.0786 0.2511>
   rotation=<rot -0.0002 90.6876 -89.3242>
  }]
 dyn_Chain_06 [{
   shapeType=Capsule
   length=2.427
   radius=2.1466
   shapeName=dyn_Chain_06_capsule
   center=<0.3507 -0.2522 -0.0413>
   rotation=<rot 0.8617 172.7448 -53.4891>
  }]
 dyn_Chain_End [{shapeType=Capsule, length=4.25, radius=2.5, shapeName=dyn_Chain_End_capsule, center=<0 0.0048 0.4969>, rotation=<rot 89.4488 90 180>}]
}
constraints [
 {bone1Name=dyn_Chain_02, bone2Name=dyn_Chain_01, swing1LimitDegrees=50, swing2LimitDegrees=50, twistLimitDegrees=20}
 {bone1Name=dyn_Chain_03, bone2Name=dyn_Chain_02, swing1LimitDegrees=50, swing2LimitDegrees=50, twistLimitDegrees=20}
 {bone1Name=dyn_Chain_04, bone2Name=dyn_Chain_03, swing1LimitDegrees=50, swing2LimitDegrees=50, twistLimitDegrees=15}
 {bone1Name=dyn_Chain_05, bone2Name=dyn_Chain_04, swing1LimitDegrees=50, swing2LimitDegrees=50, twistLimitDegrees=15}
 {bone1Name=dyn_Chain_06, bone2Name=dyn_Chain_05, swing1LimitDegrees=11.52, swing2LimitDegrees=5.76, twistLimitDegrees=0}
 {bone1Name=dyn_Chain_00, bone2Name=Melee_TwoHanded_Handle, swing1LimitDegrees=0, swing2LimitDegrees=4.32, twistLimitDegrees=0}
 {bone1Name=dyn_Chain_01, bone2Name=dyn_Chain_00, swing1LimitDegrees=59.04, swing2LimitDegrees=34.56, twistLimitDegrees=20.08}
 {
  bone1Name=dyn_Chain_End
  bone2Name=dyn_Chain_06
  swing1LimitDegrees=45
  swing1Motion=Locked
  swing2LimitDegrees=45
  swing2Motion=Locked
  twistLimitDegrees=45
  twistMotion=Locked
 }
]
